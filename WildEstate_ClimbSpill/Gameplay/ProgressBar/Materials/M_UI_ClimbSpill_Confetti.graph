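# Translucent UI material for the ClimbSpill progress bar that draws panning confetti dots in a single tint color.
# Aspect-ratio-corrected and staggered UV branches support a particle grid of anti-aliased dots with randomized sizes.
# Sine-hash randoms, a FillChance comparison, a Progress step gradient, a nine-sliced rounded box and a mirrored circle fill provide variation and masking terms.
# Named reroutes link the branches, including an opacity-scaled mask.
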Material /WildEstate_ClimbSpill/Gameplay/ProgressBar/Materials/M_UI_ClimbSpill_Confetti
graph {
 properties {bAutomaticallySetUsageInEditor=false, blendMode=BLEND_Translucent, materialDomain=MD_UI, usedShadingModels=Default Lit}
 nodes [
  n1 = Multiply(A=NamedRerouteUsage(), B=NamedRerouteUsage())
  Multiply(
   A=MaterialFunctionCall!(A=NamedRerouteUsage(), B=ScalarParameter() {parameterName=FillChance}) {materialFunction=/FortUILibrary/MaterialFunctions/Basics/MF_UI_BooleanOperations}.IsGreater
   B=@n1
  )
  n2 = Multiply(A=ScalarParameter() {defaultValue=0.7, parameterName=Dots Opacity}, B=@n1)
  n3 = NamedRerouteUsage()
  NamedRerouteDeclaration(None=Multiply(A=Multiply(A=@n2, B=@n3), B=ScalarParameter() {defaultValue=1, parameterName=Master Opacity})) {name=Mask, nodeColor=<rgba 0 0 0 1>}
  n4 = NamedRerouteUsage()
  n5 = Sine(None=DotProduct(A=Floor(None=@n4), B=Constant2Vector() {g=45.5, r=33.2758}))
  n6 = Constant() {r=0.2}
  NamedRerouteDeclaration(None=Frac(None=Multiply(A=@n5, B=@n6))) {name=Random, nodeColor=<rgba 0 0.2031 1 1>}
  n7 = Constant()
  n8 = MaterialFunctionCall!(
   UV=NamedRerouteUsage()
   Global Offset=@n7
   Global Rotation=@n7
   Tiling=Constant() {r=1}
   Cell Offset=@n7
   Cell Scale=Constant() {r=1}
   Cell Rotation=@n7
   Random Seed
  ) {materialFunction=/FortUILibrary/MaterialFunctions/Basics/MF_UI_ParticleGrid}
  n9 = LinearInterpolate(
   A=ScalarParameter() {defaultValue=0.3, parameterName=Dot Min Size}
   B=ScalarParameter() {defaultValue=0.5, parameterName=Dot Max Size}
   Alpha=@n8.Random Columns
  ) {constA=0.2, constB=0.4}
  n10 = ScalarParameter() {defaultValue=0.5, parameterName=Dot Spacing}
  n11 = Constant() {r=1}
  n12 = Distance(
   A=MaterialFunctionCall(UVs=@n8.Cells UV, Texture Scale=AppendVector(A=@n10, B=@n11)) {materialFunction=&ScaleUVsByCenter}.UVs
   B=Constant() {r=0.5}
  )
  NamedRerouteDeclaration(None=OneMinus(None=MaterialFunctionCall!(Threshold=@n9, Gradient=@n12, Filter Width Override, Filter Width Fudge Factor) {materialFunction=/Game/UI/Materials/AAThreshold}.Result)) {name=Dots, nodeColor=<rgba 0 0 0 1>}
  n13 = NamedRerouteUsage()
  n14 = Sine(None=DotProduct(A=Floor(None=@n13), B=Constant2Vector() {g=13.1, r=8.3}))
  n15 = Constant() {r=0.5}
  NamedRerouteDeclaration(None=Frac(None=Multiply(A=@n14, B=@n15))) {name=Random2, nodeColor=<rgba 0 0.2031 1 1>}
  NamedRerouteDeclaration(None=VectorParameter() {parameterName=Col1, defaultValue=<rgba 0.1946 0.0037 0 1>}.RGB) {name=Color, nodeColor=<rgba 0 0 0 1>}
  NamedRerouteDeclaration(None=Divide(A=ScalarParameter() {defaultValue=640, parameterName=Width}, B=ScalarParameter() {defaultValue=32, parameterName=Height})) {name=AspectRatio, nodeColor=<rgba 1 0 0.6094 1>}
  n16 = TextureCoordinate()
  n17 = ComponentMask(None=@n16) {r=true}
  n18 = NamedRerouteUsage()
  NamedRerouteDeclaration(None=AppendVector(A=Multiply(A=@n17, B=@n18), B=ComponentMask(None=@n16) {g=true})) {name=UV_Fixed, nodeColor=<rgba 0 0.9297 1 1>}
  n19 = MaterialFunctionCall(
   UVs=MaterialFunctionCall!(UVs) {materialFunction=/FortUILibrary/MaterialFunctions/Basics/MF_UI_9SliceUVs}.UVs
   Size=Constant2Vector() {g=0.75, r=0.75}
   CornerRoundness=Constant() {r=0.15}
  ) {materialFunction=&MF_UI_Box}
  NamedRerouteDeclaration(None=@n19.Fill) {name=BossHealthBar_Progress, nodeColor=<rgba 0 0 0 1>}
  n20 = Subtract(
   A=ComponentMask(None=MaterialFunctionCall() {materialFunction=&GetUserInterfaceUV}.Normalized UV) {r=true}
   B=ScalarParameter() {defaultValue=1, group=Progress, parameterName=Progress, sliderMax=1}
  )
  NamedRerouteDeclaration(None=OneMinus(None=Step(X=@n20) {constX=0})) {name=ProgressGradient, nodeColor=<rgba 0 0 0 1>}
  n21 = NamedRerouteUsage()
  n22 = MaterialFunctionCall!(
   UV=MaterialFunctionCall(UVs=Frac(None=@n21), "Rotation Angle (0-1)"=Constant() {r=0.25}) {materialFunction=&CustomRotator}.Rotated Values
   Mirror Y
   Swizzle UV
  ) {materialFunction=/FortUILibrary/MaterialFunctions/Basics/MF_UI_Symmetry_2x1_New}
  n23 = Add(A=ComponentMask(None=@n22.Symmetry) {r=true}) {constB=0.1}
  n24 = ComponentMask(None=@n22.Symmetry) {g=true}
  n25 = Add(A=Divide(A=NamedRerouteUsage()) {constB=4}) {constB=0.5}
  NamedRerouteDeclaration(None=MaterialFunctionCall(UVs=AppendVector(A=@n23, B=@n24), Size=Saturate(None=@n25)) {materialFunction=&MF_UI_Circle}.Fill) {name=Fill, nodeColor=<rgba 0 0 0 1>}
  n26 = MaterialFunctionCall(UV=Frac(None=Panner() {speedX=-0.05})) {materialFunction=&MF_UI_UVScaling}
  n27 = Multiply(A=Step(Y=ComponentMask(None=@n26.Scaled to Fit) {g=true}) {constX=0.5}) {constB=0.5}
  n28 = Add(A=@n26.Scaled to Fit, B=Constant2Vector() {g=0.5})
  n29 = Multiply(A=ComponentMask(None=@n28) {r=true}, B=ScalarParameter() {defaultValue=0.5, parameterName=Dot Spacing}) {constB=0.2}
  NamedRerouteDeclaration(None=AppendVector(A=Add(A=@n27, B=@n29), B=ComponentMask(None=@n28) {g=true})) {name=DotsUV, nodeColor=<rgba 0 0 0 1>}
  NamedRerouteUsage()
  NamedRerouteUsage()
  NamedRerouteUsage()
  OneMinus(None=@n19.Fill)
 ]
}
